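; State Tree task for Minor Spire scenarios that keeps the spire's objective progress info in sync.
; Enter State and every Tick mark the objective active, while Exit State marks it inactive.
; Each call stores the flag, casts the input spire reference to BGA_Spire_Minor, and passes it with current and max objective counts to UpdateObjectiveInfoServer.
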(event EventEnterState
  (params (Transition :type State%20Tree%20Transition%20Result%20Structure%20%28by%20ref%29))
  (|CallParentFunction Transition)
  (CallFunction|UpdateObjectiveonMinorSpire true))

(event EventTick
  (params (DeltaTime :type Float%20%28single-precision%29))
  (|CallParentFunction DeltaTime)
  (CallFunction|UpdateObjectiveonMinorSpire true))

(event EventExitState
  (params (Transition :type State%20Tree%20Transition%20Result%20Structure%20%28by%20ref%29))
  (|CallParentFunction Transition)
  (CallFunction|UpdateObjectiveonMinorSpire false))

(fn UpdateObjectiveOnMinorSpire
  (params (bIsObjectiveActive :type Boolean))
  (Variables|Input|SetbIsObjectiveActive bIsObjectiveActive)
  (bind _asbga_spire_minor (Utilities|Casting|CastToBGA_Spire_Minor (Variables|Input|GetMinorSpireRef)))
  (UI|UpdateObjectiveInfoServer _asbga_spire_minor (Variables|Input|GetbIsObjectiveActive) (Variables|Input|GetCurrentObjectiveCount) (Variables|Input|GetMaxObjectiveCount)))
